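# Material function for player character materials that returns a blurred copy of the scene color behind the surface.
# It averages sixteen scene color samples at fixed scattered disc offsets, scaled by Blur Size and a parameter B divided by scene depth times tan half FOV.
# On the mobile shading path it skips the blur, outputting white lerped toward Blur color by the Mask input.
# An eight-sample average and a feature level switch are left unconnected.
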MaterialFunction /BRCosmetics/Characters/Player/Common/SharedMaterials/SceneColor_Blur
graph {nodes [
  Custom!(None) {code="return GetTanFieldOfView();", description=TanHalfFieldOfView, outputType=CMOT_Float1}
  n1 = Constant2Vector() {g=-0.7935, r=0.4036}
  n2 = Custom() {code="return GetTanHalfFieldOfView();", description=TanHalfFieldOfView, outputType=CMOT_Float1}
  n3 = ComponentMask(None=SceneTexture() {sceneTextureId=PPI_SceneDepth}.Color) {r=true}
  n4 = Multiply(
   A=Divide(A=ScalarParameter() {defaultValue=0.2, parameterName=B}, B=Multiply(A=@n2, B=@n3)) {constB=1}
   B=ScalarParameter() {defaultValue=0.2, parameterName=Blur Size}
  )
  n5 = Constant2Vector() {g=-0.273, r=0.373}
  n6 = Add(
   A=SceneColor!(OffsetFraction=Multiply(A=@n1, B=@n4)) {inputMode=OffsetFraction}.RGB
   B=SceneColor(OffsetFraction=Multiply(A=@n5, B=@n4)) {inputMode=OffsetFraction}.RGB
  )
  n7 = SceneColor(OffsetFraction=Multiply(A=Constant2Vector() {g=-0.597, r=-0.041}, B=@n4)) {inputMode=OffsetFraction}
  n8 = Multiply(A=Constant2Vector() {g=-0.918, r=-0.3709}, B=@n4)
  n9 = Constant2Vector() {g=-0.335, r=-0.379}
  n10 = Add(
   A=Add(A=Add(A=@n6, B=@n7.RGB), B=SceneColor(OffsetFraction=@n8) {inputMode=OffsetFraction}.RGB)
   B=SceneColor(OffsetFraction=Multiply(A=@n9, B=@n4)) {inputMode=OffsetFraction}.RGB
  )
  n11 = SceneColor(OffsetFraction=Multiply(A=Constant2Vector() {g=0.107, r=-0.3154}, B=@n4)) {inputMode=OffsetFraction}
  n12 = Multiply(A=Constant2Vector() {g=-0.614, r=-0.767}, B=@n4)
  n13 = Constant2Vector() {g=0.639, r=-0.487}
  n14 = Add(
   A=Add(A=Add(A=@n10, B=@n11.RGB), B=SceneColor(OffsetFraction=@n12) {inputMode=OffsetFraction}.RGB)
   B=SceneColor(OffsetFraction=Multiply(A=@n13, B=@n4)) {inputMode=OffsetFraction}.RGB
  )
  Divide(A=@n14) {constB=8}
  n15 = SceneColor(OffsetFraction=Multiply(A=Constant2Vector() {g=0.151, r=0.236}, B=@n4)) {inputMode=OffsetFraction}
  n16 = Multiply(A=Constant2Vector() {g=0.731, r=0.031}, B=@n4)
  n17 = Constant2Vector() {g=-0.106, r=-0.823}
  n18 = Add(
   A=Add(A=Add(A=@n14, B=@n15.RGB), B=SceneColor(OffsetFraction=@n16) {inputMode=OffsetFraction}.RGB)
   B=SceneColor(OffsetFraction=Multiply(A=@n17, B=@n4)) {inputMode=OffsetFraction}.RGB
  )
  n19 = SceneColor(OffsetFraction=Multiply(A=Constant2Vector() {g=0.303, r=-0.8732}, B=@n4)) {inputMode=OffsetFraction}
  n20 = Multiply(A=Constant2Vector() {g=-0.116, r=0.927}, B=@n4)
  n21 = Constant2Vector() {g=0.377, r=0.698}
  n22 = Add(
   A=Add(A=Add(A=@n18, B=@n19.RGB), B=SceneColor(OffsetFraction=@n20) {inputMode=OffsetFraction}.RGB)
   B=SceneColor(OffsetFraction=Multiply(A=@n21, B=@n4)) {inputMode=OffsetFraction}.RGB
  )
  n23 = SceneColor(OffsetFraction=Multiply(A=Constant2Vector() {g=-0.515, r=0.824}, B=@n4)) {inputMode=OffsetFraction}
  n24 = Multiply(A=Constant2Vector() {g=0.753, r=0.508}, B=@n4)
  n25 = Divide(A=Add(A=Add(A=@n22, B=@n23.RGB), B=SceneColor(OffsetFraction=@n24) {inputMode=OffsetFraction}.RGB)) {constB=16}
  n26 = LinearInterpolate(
   B=VectorParameter() {parameterName=Blur color, defaultValue=<rgba 1 1 1 0>}.RGB
   Alpha=FunctionInput() {bUsePreviewValueAsDefault=true, inputName=Mask, inputType=FunctionInput_Scalar}
  ) {constA=1}
  FeatureLevelSwitch(Default=@n25, ES2=@n26, ES3_1=@n26, SM4_REMOVED=@n25, SM5=@n25, SM6=@n25)
  FunctionOutput(None=ShadingPathSwitch(Default=@n25, Deferred=@n25, Mobile=@n26))
 ]}
